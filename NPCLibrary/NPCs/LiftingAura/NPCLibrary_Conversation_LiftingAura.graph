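# Conversation database for the LiftingAura NPC, a Medic Specialist who speaks as a Greek god at odds with their father, Zeus.
# The entry point leads to one greeting speech that picks among contextual lines gated by first-ever conversation, the player's outfit, or match conversation count, with priority curves for first-meeting and outfit reactions.
# The greeting then branches through link nodes to the shared Quests and BasicServices conversations and the NPC's own Dialogue entry.
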ConversationDatabase /NPCLibrary/NPCs/LiftingAura/NPCLibrary_Conversation_LiftingAura
entry_points [{tag=Conversation.Entry.LiftingAura, destinations [#1]}]
nodes [
 ConversationEntryPointNode_7 {
  id=#1
  class=ConversationEntryPointNode
  outputs [#2]
  properties {entryTag {tagName=Conversation.Entry.LiftingAura}}
 }
 Speech_C_4 {
  id=#2
  class=Speech_C
  outputs [#3, #4, #5]
  properties {
   nodeName=""
   generalConfig {
    defaultMessage=What can the god of the squad do for you today?
    contextualMessages [
     {
      message="Hmm? Oh! Do you need any help?"
      contextRequirements [{requirement {class=FortControllerRequirement_IsFirstEverConversationWithNPC}}]
      priority {curve {rowName=Tandem.Speech.Priority.FirstEverInteraction}}
     }
     {
      message=Listen to your heart. It will tell you the truth of what you truly are...
      contextRequirements [{requirement {
         class=FortControllerRequirement_HasCID
         properties {allowedCIds [&Character_VoidRedemption_Rebel]}
        }}]
      priority {curve {rowName=Tandem.Speech.Priority.OutfitReaction}}
     }
     {
      message="Why are you doing this, father? Mortals deserve a fair chance."
      contextRequirements [{requirement {
         class=FortControllerRequirement_HasCID
         properties {allowedCIds [&Character_StunningMountain_Armor]}
        }}]
      priority {curve {rowName=Tandem.Speech.Priority.OutfitReaction}}
     }
     {
      message=You've always valued mortals. In your own way.
      contextRequirements [{requirement {
         class=FortControllerRequirement_HasCID
         properties {allowedCIds [/BRCosmetics/Athena/Items/Cosmetics/Characters/Character_DarkStance_Inferno]}
        }}]
      priority {curve {rowName=Tandem.Speech.Priority.OutfitReaction}}
     }
     {
      message="I wish we could see eye to eye, because our goals are so similar."
      contextRequirements [{requirement {
         class=FortControllerRequirement_HasCID
         properties {allowedCIds [/BRCosmetics/Athena/Items/Cosmetics/Characters/Character_ReptilianOcean_Sleek]}
        }}]
      priority {curve {rowName=Tandem.Speech.Priority.OutfitReaction}}
     }
     {
      message="What do you think, Cerberus? Can we reason with my father?"
      contextRequirements [{requirement {
         class=FortControllerRequirement_HasCID
         properties {allowedCIds [/BRCosmetics/Athena/Items/Cosmetics/Characters/Character_GateHound_Spectral]}
        }}]
      priority {curve {rowName=Tandem.Speech.Priority.OutfitReaction}}
     }
     {
      message=What can the god of the squad do for you today?
      contextRequirements [{requirement {
         class=FortControllerRequirement_CheckNPCMatchConversationCount
         properties {nPCUniqueIdTag {tagName=AISpawnerData.Type.Specialist.Medic.LiftingAura}}
        }}]
      priority {curve {rowName=None}}
     }
     {
      message=I hope you have some ideas for handling Zeus.
      contextRequirements [{requirement {
         class=FortControllerRequirement_CheckNPCMatchConversationCount
         properties {nPCUniqueIdTag {tagName=AISpawnerData.Type.Specialist.Medic.LiftingAura}}
        }}]
      priority {curve {rowName=None}}
     }
     {
      message=One day I'll make everyone get along. I hope.
      contextRequirements [{requirement {
         class=FortControllerRequirement_CheckNPCMatchConversationCount
         properties {nPCUniqueIdTag {tagName=AISpawnerData.Type.Specialist.Medic.LiftingAura}}
        }}]
      priority {curve {rowName=None}}
     }
     {
      message=You mortals deserve better than my father's wrath.
      contextRequirements [{requirement {
         class=FortControllerRequirement_CheckNPCMatchConversationCount
         properties {nPCUniqueIdTag {tagName=AISpawnerData.Type.Specialist.Medic.LiftingAura}}
        }}]
      priority {curve {rowName=None}}
     }
     {
      message="Mortals have made such wondrous things. Have you tried \"ice cream\"?"
      contextRequirements [{requirement {
         class=FortControllerRequirement_CheckNPCMatchConversationCount
         properties {nPCUniqueIdTag {tagName=AISpawnerData.Type.Specialist.Medic.LiftingAura}}
        }}]
      priority {curve {rowName=None}}
     }
    ]
   }
  }
 }
 ConversationLinkNode_3 {
  id=#5
  class=ConversationLinkNode
  properties {remoteEntryTag {tagName=Conversation.Entry.LiftingAura.Dialogue}}
 }
 ConversationLinkNode_0 {id=#4, class=ConversationLinkNode, properties {remoteEntryTag {tagName=Conversation.Entry.BasicServices}}}
 ConversationLinkNode_1 {id=#3, class=ConversationLinkNode, properties {remoteEntryTag {tagName=Conversation.Entry.NPCLibrary.Quests}}}
]
