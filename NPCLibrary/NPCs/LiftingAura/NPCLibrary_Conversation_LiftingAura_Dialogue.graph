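# Conversation database for the LiftingAura NPC's lore dialogue, in which a child of Zeus who advocates for mortals talks about the gods.
# An entry point leads to a question offering three choices: Mortals, Zeus's Past, or Back.
# Each topic runs a linear chain of speech nodes, then a ReturnToStart node labeled BACK.
# The first node carries an "Ask about..." choice label.
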ConversationDatabase /NPCLibrary/NPCs/LiftingAura/NPCLibrary_Conversation_LiftingAura_Dialogue
entry_points [{tag=Conversation.Entry.LiftingAura.Dialogue, destinations [#1]}]
nodes [
 ConversationEntryPointNode_0 {
  id=#1
  class=ConversationEntryPointNode
  outputs [#2]
  properties {entryTag {tagName=Conversation.Entry.LiftingAura.Dialogue}}
 }
 Speech_C_0 {
  id=#2
  class=Speech_C
  outputs [#3, #4, #5]
  properties {
   nodeName=""
   softTaskIcon=&T_UI_NPC_question
   generalConfig {defaultMessage=What would you like to know more about?}
  }
  sub_nodes [ConversationChoiceNode_5 {class=ConversationChoiceNode, properties {defaultChoiceDisplayText=Ask about...}}]
 }
 Speech_C_4 {
  id=#5
  class=Speech_C
  outputs [#6]
  properties {
   nodeName=""
   softTaskIcon=&T_UI_NPC_question
   generalConfig {defaultMessage="He wasn't always like this. My father, I mean. Zeus."}
  }
  sub_nodes [ConversationChoiceNode_3 {class=ConversationChoiceNode, properties {defaultChoiceDisplayText=Zeus's Past}}]
 }
 Speech_C_5 {
  id=#6
  class=Speech_C
  outputs [#7]
  properties {
   nodeName=""
   softTaskIcon=&T_UI_NPC_question
   generalConfig {defaultMessage="I remember, eons ago, he would give mortals gifts!"}
  }
 }
 Speech_C_6 {
  id=#7
  class=Speech_C
  outputs [#8]
  properties {
   nodeName=""
   softTaskIcon=&T_UI_NPC_question
   generalConfig {defaultMessage=Then something...changed in him. He's grown more distant ever since.}
  }
 }
 Speech_C_10 {
  id=#8
  class=Speech_C
  outputs [#9]
  properties {
   nodeName=""
   softTaskIcon=&T_UI_NPC_question
   generalConfig {defaultMessage="I've kept peace between him and the mortals, but now..."}
  }
 }
 Speech_C_2 {
  id=#9
  class=Speech_C
  outputs [#10]
  properties {
   nodeName=""
   softTaskIcon=&T_UI_NPC_question
   generalConfig {defaultMessage=Now I worry no one can stop his wrath.}
  }
 }
 ReturnToStart_C_0 {
  id=#10
  class=ReturnToStart_C
  properties {softTaskIcon=&T_UI_NPC_Back}
  sub_nodes [ConversationChoiceNode_1 {class=ConversationChoiceNode, properties {defaultChoiceDisplayText=BACK}}]
 }
 Back_C_3 {
  id=#4
  class=Back_C
  sub_nodes [ConversationChoiceNode_23 {class=ConversationChoiceNode, properties {defaultChoiceDisplayText=BACK}}]
 }
 Speech_C_3 {
  id=#3
  class=Speech_C
  outputs [#11]
  properties {
   nodeName=""
   softTaskIcon=&T_UI_NPC_question
   generalConfig {defaultMessage=I don't really know how I became the advocate for mortals among the gods.}
  }
  sub_nodes [ConversationChoiceNode_0 {class=ConversationChoiceNode, properties {defaultChoiceDisplayText=Mortals}}]
 }
 Speech_C_1 {
  id=#11
  class=Speech_C
  outputs [#12]
  properties {
   nodeName=""
   softTaskIcon=&T_UI_NPC_question
   generalConfig {defaultMessage=Things just didn't seem... fair. And I made sure the other gods knew that.}
  }
 }
 Speech_C_9 {
  id=#12
  class=Speech_C
  outputs [#13]
  properties {
   nodeName=""
   softTaskIcon=&T_UI_NPC_question
   generalConfig {defaultMessage="After a while, I guess I developed a reputation."}
  }
 }
 Speech_C_7 {
  id=#13
  class=Speech_C
  outputs [#14]
  properties {
   nodeName=""
   softTaskIcon=&T_UI_NPC_question
   generalConfig {defaultMessage=I don't regret it. I'm proud of standing up for my beliefs.}
  }
 }
 ReturnToStart_C_1 {
  id=#14
  class=ReturnToStart_C
  properties {softTaskIcon=&T_UI_NPC_Back}
  sub_nodes [ConversationChoiceNode_2 {class=ConversationChoiceNode, properties {defaultChoiceDisplayText=BACK}}]
 }
]
